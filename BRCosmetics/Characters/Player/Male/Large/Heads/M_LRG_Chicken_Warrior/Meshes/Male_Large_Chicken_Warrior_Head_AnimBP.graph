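# Animation Blueprint for the large male Chicken Warrior cosmetic head that feeds orientation data to its material.
# Each update calls the parent, then continues only when rendering material quality is above 0.
# It converts a zero world transform into the part mesh's root bone space, building an inverse-root transform.
# That transform rotates the three unit axes, written to the dynamic material's ActorRotationX, Y and Z vector parameters.
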AnimBlueprint /BRCosmetics/Characters/Player/Male/Large/Heads/M_LRG_Chicken_Warrior/Meshes/Male_Large_Chicken_Warrior_Head_AnimBP
target_skeleton=/Game/Characters/Player/Common/Fortnite_Base_Head/Fortnite_Base_Head_Export_Skeleton
graphs [EventGraph {nodes [
   n1 = Event BlueprintUpdateAnimation()
   > Parent: BlueprintUpdateAnimation!(execute, DeltaTimeX:float=@n1.DeltaTimeX) -> (then) {is=title K2Node_CallParentFunction}
   > Branch(Condition="integer > integer"(A=GetRenderingMaterialQualityLevel!() -> (ReturnValue:int=0), B=0) {is=title K2Node_CallFunction})
   n2 = GetPartSkeletalMeshComponent()
   > n3 = Get Dynamic Material Instance(SkeletalMeshComponent=@n2)
   n4 = TransformToBoneSpace!(self:Skinned Mesh Component Object Reference=@n2, BoneName:Name=root, InPosition:Vector="0, 0, 0", InRotation:Rotator="0, 0, 0") -> (OutPosition:Vector, OutRotation:Rotator)
   n5 = MakeTransform(Location=@n4, Rotation=@n4.OutRotation)
   > SetVectorParameterValue(
    self=@n3
    ParameterName=ActorRotationX
    Value="To LinearColor (Vector)"!(InVec:Vector=TransformDirection!(T:"Transform (by ref)"=@n5, Direction:Vector="1,0, 0") -> (ReturnValue:Vector)) -> (ReturnValue:Linear Color Structure)
   )
   > SetVectorParameterValue(self=@n3, ParameterName=ActorRotationY, Value="To LinearColor (Vector)"(InVec=TransformDirection(T=@n5, Direction="0,1, 0")))
   > SetVectorParameterValue(self=@n3, ParameterName=ActorRotationZ, Value="To LinearColor (Vector)"(InVec=TransformDirection(T=@n5, Direction="0, 0,1")))
   TryGetPawnOwner()
  ]}]
variables []
